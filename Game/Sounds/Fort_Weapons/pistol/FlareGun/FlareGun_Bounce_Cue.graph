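# Sound cue for the bounce sound of flare gun projectiles.
# A Random node picks among six equally weighted bounce wave variations, passed through a Modulator with pitch locked at 1.
# It uses grenade bounce attenuation and the ranged weapon projectile effects sound class at boosted volume.
# A concurrency override caps simultaneous instances at a few and blocks new ones beyond that.
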SoundCue /Game/Sounds/Fort_Weapons/pistol/FlareGun/FlareGun_Bounce_Cue
root=Modulator_5
nodes [
 {id=Modulator_5, children [Random_1], properties {pitchMax=1, pitchMin=1}}
 {
  id=Random_1
  children [WavePlayer_6, WavePlayer_7, WavePlayer_8, WavePlayer_9, WavePlayer_10, WavePlayer_11]
  properties {weights [1,1,1,1,1,1]}
 }
 {id=WavePlayer_10, properties {soundWaveAssetPtr=&/FlareGun_Bounce_04}}
 {id=WavePlayer_11, properties {soundWaveAssetPtr=&/FlareGun_Bounce_06}}
 {id=WavePlayer_6, properties {soundWaveAssetPtr=&/FlareGun_Bounce_01}}
 {id=WavePlayer_7, properties {soundWaveAssetPtr=&/FlareGun_Bounce_05}}
 {id=WavePlayer_8, properties {soundWaveAssetPtr=&/FlareGun_Bounce_02}}
 {id=WavePlayer_9, properties {soundWaveAssetPtr=&/FlareGun_Bounce_03}}
]
properties {
 attenuationSettings=/Game/Sounds/Attenuation/Fort_Grenade_Bounce_Att
 bOverrideConcurrency=true
 maxDistance=5500
 soundClassObject=&Fort_WeaponRanged_Projectile_Effects
 volumeMultiplier=1.8
 concurrencyOverrides {maxCount=3, resolutionRule=PreventNew}
}
