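# Sound cue for flare gun ember impacts, using the third-person ranged weapon fire sound class.
# A weighted random node picks one of six ember impact waves and feeds a modulator at the root.
# The cue raises pitch slightly and lowers volume, with its own logarithmic distance attenuation.
# A dedicated concurrency asset governs overlapping instances.
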SoundCue /Game/Sounds/Fort_Weapons/pistol/FlareGun/FlareGun_EmberImpact_Cue
root=Modulator_2
nodes [
 {id=Modulator_2, children [Random_3]}
 {
  id=Random_3
  children [WavePlayer_0, WavePlayer_1, WavePlayer_4, WavePlayer_10, WavePlayer_11, WavePlayer_12]
  properties {weights [1,2,1,1,1,2]}
 }
 {id=WavePlayer_0, properties {soundWaveAssetPtr=&/FlareGun_EmberImpact_01}}
 {id=WavePlayer_1, properties {soundWaveAssetPtr=&/FlareGun_EmberImpact_02}}
 {id=WavePlayer_10, properties {soundWaveAssetPtr=&/FlareGun_EmberImpact_04}}
 {id=WavePlayer_11, properties {soundWaveAssetPtr=&/FlareGun_EmberImpact_05}}
 {id=WavePlayer_12, properties {soundWaveAssetPtr=&/FlareGun_EmberImpact_06}}
 {id=WavePlayer_4, properties {soundWaveAssetPtr=&/FlareGun_EmberImpact_03}}
]
properties {
 bOverrideAttenuation=true
 maxDistance=2500
 pitchMultiplier=1.1
 soundClassObject=&Fort_WeaponRanged_Fire_3P_PSM
 volumeMultiplier=0.7
 attenuationOverrides {distanceAlgorithm=Logarithmic, falloffDistance=2000, reverbWetLevelMin=0.2, stereoSpread=800, attenuationShapeExtents {x=500}}
 concurrencyOverrides {maxCount=4, resolutionRule=StopOldest, volumeScale=0.5}
 concurrencySet [&/FlareGun_EmberImpact_Conc]
}
