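# Sound cue for flare gun embers hitting water.
# A Random node picks one of five equally weighted water impact waves and feeds a Modulator that varies each play.
# The cue slightly raises pitch, lowers volume and overrides attenuation with logarithmic falloff.
# It uses the third-person ranged weapon fire sound class and a shared ember impact concurrency set to manage overlapping impacts.
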SoundCue /Game/Sounds/Fort_Weapons/pistol/FlareGun/FlareGun_EmberImpact_Water_Cue
root=Modulator_2
nodes [
 {id=Modulator_2, children [Random_0]}
 {id=Random_0, children [WavePlayer_2,WavePlayer_3,WavePlayer_5,WavePlayer_6,WavePlayer_7], properties {weights [1,1,1,1,1]}}
 {id=WavePlayer_2, properties {soundWaveAssetPtr=&/FlareGun_EmberImpact_Water_01}}
 {id=WavePlayer_3, properties {soundWaveAssetPtr=&/FlareGun_EmberImpact_Water_02}}
 {id=WavePlayer_5, properties {soundWaveAssetPtr=&/FlareGun_EmberImpact_Water_03}}
 {id=WavePlayer_6, properties {soundWaveAssetPtr=&/FlareGun_EmberImpact_Water_04}}
 {id=WavePlayer_7, properties {soundWaveAssetPtr=&/FlareGun_EmberImpact_Water_05}}
]
properties {
 bOverrideAttenuation=true
 maxDistance=2500
 pitchMultiplier=1.1
 soundClassObject=&Fort_WeaponRanged_Fire_3P_PSM
 volumeMultiplier=0.8
 attenuationOverrides {distanceAlgorithm=Logarithmic, falloffDistance=2000, reverbWetLevelMin=0.2, stereoSpread=800, attenuationShapeExtents {x=500}}
 concurrencyOverrides {maxCount=4, resolutionRule=StopOldest, volumeScale=0.5}
 concurrencySet [&/FlareGun_EmberImpact_Conc]
}
