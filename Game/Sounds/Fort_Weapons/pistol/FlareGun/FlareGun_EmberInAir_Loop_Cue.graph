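# Looping spatialized sound cue for a flare gun ember in the air, routed to the third-person ranged weapon fire sound class.
# A Random node chooses among three equally weighted looping wave variations and feeds the root Attenuation node.
# That node overrides its referenced projectile attenuation asset with inline logarithmic falloff and HRTF spatialization.
# A shared explosive-fuse concurrency set governs simultaneous instances, and the cue's overall volume is reduced.
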SoundCue /Game/Sounds/Fort_Weapons/pistol/FlareGun/FlareGun_EmberInAir_Loop_Cue
root=Attenuation_3
nodes [
 {
  id=Attenuation_3
  children [Random_0]
  properties {
   attenuationSettings=&Fort_Weapon_Projectile_Medium_Mono_Athena_Att
   bOverrideAttenuation=true
   attenuationOverrides {
    distanceAlgorithm=Logarithmic
    falloffDistance=1536
    reverbWetLevelMin=0.2
    spatializationAlgorithm=SPATIALIZATION_HRTF
    stereoSpread=800
    attenuationShapeExtents {x=512}
   }
  }
 }
 {id=Random_0, children [WavePlayer_4,WavePlayer_5,WavePlayer_6], properties {weights [1,1,1]}}
 {id=WavePlayer_4, properties {bLooping=true, soundWaveAssetPtr=&/FlareGun_EmberInAir_Loop_01}}
 {id=WavePlayer_5, properties {bLooping=true, soundWaveAssetPtr=&/FlareGun_EmberInAir_Loop_02}}
 {id=WavePlayer_6, properties {bLooping=true, soundWaveAssetPtr=&/FlareGun_EmberInAir_Loop_03}}
]
properties {
 maxDistance=2048
 soundClassObject=&Fort_WeaponRanged_Fire_3P_PSM
 volumeMultiplier=0.4
 attenuationOverrides {falloffDistance=5000, attenuationShapeExtents {x=1000}}
 concurrencyOverrides {
  bVolumeScaleCanRelease=true
  maxCount=2
  resolutionRule=StopFarthestThenPreventNew
  volumeScale=0.2
  volumeScaleAttackTime=0.2
  volumeScaleReleaseTime=2
 }
 concurrencySet [/Game/Sounds/Concurrency_Rules/Athena_Explosive_Fuse_Conc]
}
