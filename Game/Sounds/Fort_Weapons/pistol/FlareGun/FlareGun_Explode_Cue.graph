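# Sound cue for the flare gun explosion, layering close and distant sounds under a weapon explosion concurrency set.
# Each mixer blends random Close waves with quieter Medium waves, each layer having its own sound class and attenuation.
# Modulators add variation, with a wider pitch range on the distant layer.
# The quality node gives its first level four waves per layer, while the other three share a two-wave mix.
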SoundCue /Game/Sounds/Fort_Weapons/pistol/FlareGun/FlareGun_Explode_Cue
root=QualityLevel_0
nodes [
 {
  id=Attenuation_0
  children [Modulator_1]
  properties {
   attenuationSettings=/Game/Sounds/Attenuation/Fort_Weapon_Rocket_Distant_Explosion_Athena_Att_2
   attenuationOverrides {falloffDistance=3000, attenuationShapeExtents {x=1000}}
  }
 }
 {
  id=Attenuation_3
  children [Random_0]
  properties {attenuationSettings=/Game/Sounds/Attenuation/Fort_Weapon_Rocket_Medium_Explosion_Athena_Att_2, attenuationOverrides {falloffDistance=1600}}
 }
 {
  id=Attenuation_6
  children [Modulator_5]
  properties {
   attenuationSettings=&Fort_Weapon_Rocket_Distant_Explosion_Athena_Att_2
   attenuationOverrides {falloffDistance=3000, attenuationShapeExtents {x=1000}}
  }
 }
 {
  id=Attenuation_7
  children [Random_3]
  properties {attenuationSettings=&Fort_Weapon_Rocket_Medium_Explosion_Athena_Att_2, attenuationOverrides {falloffDistance=1600}}
 }
 {id=Mixer_2, children [SoundClass_5,SoundClass_0], properties {inputVolume [0.8,0.5]}}
 {id=Mixer_4, children [SoundClass_3,SoundClass_2], properties {inputVolume [0.8,0.5]}}
 {id=Modulator_0, children [Attenuation_7]}
 {id=Modulator_1, children [Random_1], properties {pitchMax=1.1, pitchMin=0.9}}
 {id=Modulator_4, children [Attenuation_3]}
 {id=Modulator_5, children [Random_4], properties {pitchMax=1.1, pitchMin=0.9}}
 {id=QualityLevel_0, children [Mixer_4,Mixer_2,Mixer_2,Mixer_2]}
 {id=Random_0, children [WavePlayer_2,WavePlayer_6], properties {weights [1,1]}}
 {id=Random_1, children [WavePlayer_8, WavePlayer_15], properties {weights [1,1]}}
 {id=Random_3, children [WavePlayer_0, WavePlayer_1, WavePlayer_4, WavePlayer_10], properties {weights [1,1,1,1]}}
 {id=Random_4, children [WavePlayer_11, WavePlayer_12, WavePlayer_13, WavePlayer_14], properties {weights [1,1,1,1]}}
 {
  id=SoundClass_0
  children [Attenuation_0]
  properties {soundClassOverride=&Fort_WeaponRanged_Fire_Distant}
 }
 {
  id=SoundClass_2
  children [Attenuation_6]
  properties {soundClassOverride=&Fort_WeaponRanged_Fire_Distant}
 }
 {id=SoundClass_3, children [Modulator_0], properties {soundClassOverride=&Fort_WeaponRanged_Fire_3P_PSM}}
 {id=SoundClass_5, children [Modulator_4], properties {soundClassOverride=&Fort_WeaponRanged_Fire_3P_PSM}}
 {id=WavePlayer_0, properties {soundWaveAssetPtr=&/FlareGun_Explode_Close_01}}
 {id=WavePlayer_1, properties {soundWaveAssetPtr=&/FlareGun_Explode_Close_02}}
 {id=WavePlayer_10, properties {soundWaveAssetPtr=&/FlareGun_Explode_Close_04}}
 {id=WavePlayer_11, properties {soundWaveAssetPtr=&/FlareGun_Explode_Medium_01}}
 {id=WavePlayer_12, properties {soundWaveAssetPtr=&/FlareGun_Explode_Medium_02}}
 {id=WavePlayer_13, properties {soundWaveAssetPtr=&/FlareGun_Explode_Medium_03}}
 {id=WavePlayer_14, properties {soundWaveAssetPtr=&/FlareGun_Explode_Medium_04}}
 {id=WavePlayer_15, properties {soundWaveAssetPtr=&FlareGun_Explode_Medium_04}}
 {id=WavePlayer_2, properties {soundWaveAssetPtr=&FlareGun_Explode_Close_01}}
 {id=WavePlayer_4, properties {soundWaveAssetPtr=&/FlareGun_Explode_Close_03}}
 {id=WavePlayer_6, properties {soundWaveAssetPtr=&FlareGun_Explode_Close_04}}
 {id=WavePlayer_8, properties {soundWaveAssetPtr=&FlareGun_Explode_Medium_02}}
]
properties {
 maxDistance=10000
 pitchMultiplier=0.9
 volumeMultiplier=0.7
 concurrencyOverrides {maxCount=4, resolutionRule=StopOldest, volumeScale=0.5}
 concurrencySet [/Game/Sounds/Concurrency_Rules/Weapon_Explosion_Conc]
}
