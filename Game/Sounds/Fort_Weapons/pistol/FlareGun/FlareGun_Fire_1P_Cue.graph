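# Sound cue for the first-person flare gun firing sound.
# A random node picks one of six equally weighted close-fire waves, feeding a modulator and then a mixer whose second input is empty.
# The cue slightly lowers volume and pitch, and references shared first-person weapon sound class, attenuation and gunfire concurrency assets.
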SoundCue /Game/Sounds/Fort_Weapons/pistol/FlareGun/FlareGun_Fire_1P_Cue
root=Mixer_0
nodes [
 {id=Mixer_0, children [Modulator_0,null], properties {inputVolume [1,0.3]}}
 {id=Modulator_0, children [Random_0]}
 {id=Random_0, children [WavePlayer_0,WavePlayer_1,WavePlayer_2,WavePlayer_3,WavePlayer_4,WavePlayer_5], properties {weights [1,1,1,1,1,1]}}
 {id=WavePlayer_0, properties {soundWaveAssetPtr=&/FlareGun_Fire_Close_01}}
 {id=WavePlayer_1, properties {soundWaveAssetPtr=&/FlareGun_Fire_Close_02}}
 {id=WavePlayer_2, properties {soundWaveAssetPtr=&/FlareGun_Fire_Close_03}}
 {id=WavePlayer_3, properties {soundWaveAssetPtr=&/FlareGun_Fire_Close_04}}
 {id=WavePlayer_4, properties {soundWaveAssetPtr=&/FlareGun_Fire_Close_05}}
 {id=WavePlayer_5, properties {soundWaveAssetPtr=&/FlareGun_Fire_Close_06}}
]
properties {
 attenuationSettings=&Fort_Weapon_Projectile_FirstPerson_Athena_Att
 maxDistance=2097152
 pitchMultiplier=0.9
 soundClassObject=&Fort_WeaponRanged_Fire_1P_PSM
 volumeMultiplier=0.8
 attenuationOverrides {falloffDistance=5000, attenuationShapeExtents {x=1000}}
 concurrencyOverrides {bLimitToOwner=true, maxCount=3, resolutionRule=PreventNew}
 concurrencySet [&Weapon_Gunfire_FirstPerson_Conc]
}
