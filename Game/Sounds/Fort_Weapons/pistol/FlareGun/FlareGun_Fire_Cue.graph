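# Flare gun fire sound cue that layers close, medium and distant gunshot recordings by distance.
# A mixer blends three branches, each with its own sound class override and projectile attenuation, fed by a random pick of wave variants.
# The close and distant branches pass through pitch modulators, and the random nodes are excluded from branch culling.
# Playback is limited by local and global weapon gunfire concurrency sets.
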SoundCue /Game/Sounds/Fort_Weapons/pistol/FlareGun/FlareGun_Fire_Cue
root=Mixer_3
nodes [
 {id=Attenuation_0, children [Modulator_0], properties {attenuationSettings=&Fort_Weapon_Projectile_Close_Athena_Att}}
 {
  id=Attenuation_10
  children [Random_1]
  properties {attenuationSettings=&Fort_Weapon_Projectile_Medium_Mono_Athena_Att}
 }
 {
  id=Attenuation_11
  children [Modulator_7]
  properties {attenuationSettings=&Fort_Weapon_Projectile_Distant_Mono_Athena_Att}
 }
 {id=Mixer_3, children [SoundClass_0,SoundClass_2,SoundClass_6], properties {inputVolume [0.5,0.5,0.5]}}
 {id=Modulator_0, children [Random_0], properties {pitchMax=1, pitchMin=1}}
 {id=Modulator_7, children [Random_2], properties {pitchMax=1.1}}
 {
  id=Random_0
  children [WavePlayer_0,WavePlayer_1,WavePlayer_2,WavePlayer_3,WavePlayer_4,WavePlayer_5]
  properties {bShouldExcludeFromBranchCulling=true, weights [1,1,1,1,1,1]}
 }
 {
  id=Random_1
  children [WavePlayer_6, WavePlayer_7, WavePlayer_8, WavePlayer_9, WavePlayer_10]
  properties {bShouldExcludeFromBranchCulling=true, weights [1,1,1,1,1]}
 }
 {
  id=Random_2
  children [WavePlayer_12, WavePlayer_13, WavePlayer_14, WavePlayer_15]
  properties {bShouldExcludeFromBranchCulling=true, weights [1,1,1,1]}
 }
 {
  id=SoundClass_0
  children [Attenuation_0]
  properties {soundClassOverride=&Fort_WeaponRanged_Fire_3P_PSM}
 }
 {
  id=SoundClass_2
  children [Attenuation_10]
  properties {soundClassOverride=&Fort_WeaponRanged_Fire_Distant}
 }
 {
  id=SoundClass_6
  children [Attenuation_11]
  properties {soundClassOverride=&Fort_WeaponRanged_Fire_Distant}
 }
 {id=WavePlayer_0, properties {soundWaveAssetPtr=&/FlareGun_Fire_Close_01}}
 {id=WavePlayer_1, properties {soundWaveAssetPtr=&/FlareGun_Fire_Close_02}}
 {id=WavePlayer_10, properties {soundWaveAssetPtr=&/FlareGun_Fire_Medium_05}}
 {id=WavePlayer_12, properties {soundWaveAssetPtr=&/FlareGun_Fire_Distant_02}}
 {id=WavePlayer_13, properties {soundWaveAssetPtr=&/FlareGun_Fire_Distant_03}}
 {id=WavePlayer_14, properties {soundWaveAssetPtr=&/FlareGun_Fire_Distant_04}}
 {id=WavePlayer_15, properties {soundWaveAssetPtr=&/FlareGun_Fire_Distant_05}}
 {id=WavePlayer_2, properties {soundWaveAssetPtr=&/FlareGun_Fire_Close_03}}
 {id=WavePlayer_3, properties {soundWaveAssetPtr=&/FlareGun_Fire_Close_04}}
 {id=WavePlayer_4, properties {soundWaveAssetPtr=&/FlareGun_Fire_Close_05}}
 {id=WavePlayer_5, properties {soundWaveAssetPtr=&/FlareGun_Fire_Close_06}}
 {id=WavePlayer_6, properties {soundWaveAssetPtr=&/FlareGun_Fire_Medium_01}}
 {id=WavePlayer_7, properties {soundWaveAssetPtr=&/FlareGun_Fire_Medium_02}}
 {id=WavePlayer_8, properties {soundWaveAssetPtr=&/FlareGun_Fire_Medium_03}}
 {id=WavePlayer_9, properties {soundWaveAssetPtr=&/FlareGun_Fire_Medium_04}}
]
properties {
 maxDistance=27500
 soundClassObject=&Fort_WeaponRanged_Fire_3P_PSM
 volumeMultiplier=1
 attenuationOverrides {falloffDistance=5000, attenuationShapeExtents {x=1000}}
 concurrencyOverrides {bLimitToOwner=true, maxCount=3, resolutionRule=PreventNew}
 concurrencySet [&Weapon_Gunfire_Local_Conc, &Weapon_Gunfire_Global_Conc]
}
