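# Sound cue for the looping in-air sound of a flare gun projectile.
# A Random node picks between two equally weighted looping wave players, each holding one loop variation.
# Playback uses a flare gun projectile alert attenuation asset, a third-person ranged weapon fire sound class, a reduced volume and a shared explosive fuse concurrency rule to manage simultaneous instances.
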SoundCue /Game/Sounds/Fort_Weapons/pistol/FlareGun/FlareGun_InAirLoop_Cue
root=Random_1
nodes [
 {id=Random_1, children [WavePlayer_7,WavePlayer_8], properties {weights [1,1]}}
 {id=WavePlayer_7, properties {bLooping=true, soundWaveAssetPtr=&/FlareGun_InAirLoop_01}}
 {id=WavePlayer_8, properties {bLooping=true, soundWaveAssetPtr=&/FlareGun_InAirLoop_02}}
]
properties {
 attenuationSettings=/Game/Sounds/Attenuation/Fort_Weapon_Projectile_FlareGun_Alert_Att
 maxDistance=10000
 soundClassObject=&Fort_WeaponRanged_Fire_3P_PSM
 volumeMultiplier=0.6
 attenuationOverrides {falloffDistance=5000, attenuationShapeExtents {x=1000}}
 concurrencyOverrides {
  bVolumeScaleCanRelease=true
  maxCount=2
  resolutionRule=StopFarthestThenPreventNew
  volumeScale=0.2
  volumeScaleAttackTime=0.2
  volumeScaleReleaseTime=2
 }
 concurrencySet [/Game/Sounds/Concurrency_Rules/Athena_Explosive_Fuse_Conc]
}
