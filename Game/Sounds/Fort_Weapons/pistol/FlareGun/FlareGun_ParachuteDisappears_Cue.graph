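# Sound cue for the flare gun's parachute disappearing, routed to the Fort weapons sound class.
# A random node picks one of three equally weighted wave variations, played at lowered volume and slightly lowered pitch.
# It borrows the grenade bounce attenuation asset for spatial falloff and references the PickupMaximum_Oldest concurrency set to limit simultaneous playback.
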SoundCue /Game/Sounds/Fort_Weapons/pistol/FlareGun/FlareGun_ParachuteDisappears_Cue
root=Random_0
nodes [
 {id=Random_0, children [WavePlayer_2,WavePlayer_3,WavePlayer_4], properties {weights [1,1,1]}}
 {id=WavePlayer_2, properties {soundWaveAssetPtr=&/FlareGun_ParachuteDisappears_01}}
 {id=WavePlayer_3, properties {soundWaveAssetPtr=&/FlareGun_ParachuteDisappears_02}}
 {id=WavePlayer_4, properties {soundWaveAssetPtr=&/FlareGun_ParachuteDisappears_03}}
]
properties {
 attenuationSettings=/Game/Sounds/Attenuation/Fort_Grenade_Bounce_Att
 maxDistance=5500
 pitchMultiplier=0.9
 soundClassObject=/Game/Sounds/FortSoundClassesAndModes/Fort_Weapons
 volumeMultiplier=0.2963
 concurrencyOverrides {maxCount=1, resolutionRule=PreventNew}
 concurrencySet [&PickupMaximum_Oldest]
}
