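# FlareGun_Pickup_Cue is the sound cue for picking up the flare gun weapon.
# A Random node picks one of three pickup wave variations with equal weight.
# It uses the shared weapon pickup attenuation and loot-collect sound class at reduced volume.
# A pickup concurrency set is assigned, alongside override values for a single instance that prevents new plays.
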SoundCue /Game/Sounds/Fort_Weapons/pistol/FlareGun/FlareGun_Pickup_Cue
root=Random_0
nodes [
 {id=Random_0, children [WavePlayer_2,WavePlayer_3,WavePlayer_4], properties {weights [1,1,1]}}
 {id=WavePlayer_2, properties {soundWaveAssetPtr=&/FlareGun_Pickup_01}}
 {id=WavePlayer_3, properties {soundWaveAssetPtr=&/FlareGun_Pickup_03}}
 {id=WavePlayer_4, properties {soundWaveAssetPtr=&/FlareGun_Pickup_02}}
]
properties {
 attenuationSettings=&Fort_Weapon_Pickup_Att
 maxDistance=3500
 soundClassObject=&Fort_Loot_Collect
 volumeMultiplier=0.3692
 concurrencyOverrides {maxCount=1, resolutionRule=PreventNew}
 concurrencySet [&PickupMaximum_Oldest]
}
