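# Sound cue for the end of the Flare Gun reload, played as weapon foley in Fortnite's pistol weapon set.
# A Random node picks one of three equally weighted reload-end wave variants, and a Modulator at the root feeds the output.
# It uses the shared weapon reload attenuation and the weapon foley sound class at reduced volume.
# Overridden concurrency allows one instance, stopping the farthest before blocking new plays.
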SoundCue /Game/Sounds/Fort_Weapons/pistol/FlareGun/FlareGun_ReloadEnd_Cue
root=Modulator_1
nodes [
 {id=Modulator_1, children [Random_1]}
 {id=Random_1, children [WavePlayer_3,WavePlayer_4,WavePlayer_5], properties {weights [1,1,1]}}
 {id=WavePlayer_3, properties {soundWaveAssetPtr=&/FlareGun_ReloadEnd_B01}}
 {id=WavePlayer_4, properties {soundWaveAssetPtr=&/FlareGun_ReloadEnd_B02}}
 {id=WavePlayer_5, properties {soundWaveAssetPtr=&/FlareGun_ReloadEnd_B03}}
]
properties {
 attenuationSettings=&Fort_Weapon_Reload_Att
 bOverrideConcurrency=true
 maxDistance=2250
 soundClassObject=&Fort_WeaponFoley
 volumeMultiplier=0.501
 attenuationOverrides {falloffDistance=5000, attenuationShapeExtents {x=1000}}
 concurrencyOverrides {maxCount=1, resolutionRule=StopFarthestThenPreventNew}
}
