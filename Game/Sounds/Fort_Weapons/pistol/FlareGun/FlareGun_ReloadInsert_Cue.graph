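# Sound cue for the shell-insert foley of the flare gun reload, routed to the weapon foley sound class with reload attenuation.
# A mixer layers one of three randomly picked flare gun insert waves with a quieter, pitched-up grenade launcher shell-insert wave borrowed from the launcher sounds.
# An outer modulator at default settings wraps the mix, and a concurrency override limits playback to a single instance.
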SoundCue /Game/Sounds/Fort_Weapons/pistol/FlareGun/FlareGun_ReloadInsert_Cue
root=Modulator_1
nodes [
 {id=Mixer_0, children [Random_0,Modulator_0], properties {inputVolume [1.8,0.8]}}
 {id=Modulator_0, children [WavePlayer_16], properties {pitchMax=1.2, pitchMin=1.1}}
 {id=Modulator_1, children [Mixer_0]}
 {id=Random_0, children [WavePlayer_0,WavePlayer_1,WavePlayer_2], properties {weights [1,1,1]}}
 {id=WavePlayer_0, properties {soundWaveAssetPtr=&/FlareGun_ReloadInsert_01}}
 {id=WavePlayer_1, properties {soundWaveAssetPtr=&/FlareGun_ReloadInsert_02}}
 {id=WavePlayer_16, properties {soundWaveAssetPtr=/Game/Sounds/Fort_Weapons/Launchers/grenadelauncher_reload_insert_shell}}
 {id=WavePlayer_2, properties {soundWaveAssetPtr=&/FlareGun_ReloadInsert_03}}
]
properties {
 attenuationSettings=&Fort_Weapon_Reload_Att
 bOverrideConcurrency=true
 maxDistance=2250
 soundClassObject=&Fort_WeaponFoley
 volumeMultiplier=0.351
 attenuationOverrides {falloffDistance=5000, attenuationShapeExtents {x=1000}}
 concurrencyOverrides {maxCount=1, resolutionRule=StopFarthestThenPreventNew}
}
