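# Sound cue for the start of the flare gun pistol's reload.
# A Random node picks one of three equally weighted reload-start waves and feeds it through a Modulator at the root.
# It plays in the weapon foley sound class with reload attenuation, lowered pitch and volume, and a concurrency override allowing one instance, stopping the farthest first.
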SoundCue /Game/Sounds/Fort_Weapons/pistol/FlareGun/FlareGun_ReloadStart_Cue
root=Modulator_1
nodes [
 {id=Modulator_1, children [Random_0]}
 {id=Random_0, children [WavePlayer_0,WavePlayer_1,WavePlayer_2], properties {weights [1,1,1]}}
 {id=WavePlayer_0, properties {soundWaveAssetPtr=&/FlareGun_ReloadStart_01}}
 {id=WavePlayer_1, properties {soundWaveAssetPtr=&/FlareGun_ReloadStart_02}}
 {id=WavePlayer_2, properties {soundWaveAssetPtr=&/FlareGun_ReloadStart_03}}
]
properties {
 attenuationSettings=&Fort_Weapon_Reload_Att
 bOverrideConcurrency=true
 maxDistance=2250
 pitchMultiplier=0.9
 soundClassObject=&Fort_WeaponFoley
 volumeMultiplier=0.702
 attenuationOverrides {falloffDistance=5000, attenuationShapeExtents {x=1000}}
 concurrencyOverrides {maxCount=1, resolutionRule=StopFarthestThenPreventNew}
}
