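# Looping Sound Cue for the flare gun's marking scan, using the projectile alert attenuation asset and the distant ranged-weapon fire sound class.
# A Mixer layers the scan loop with a random pick of two looping in-air waves.
# Only the scan loop passes through an Enveloper, which ramps volume up from near silence with slight pitch and volume variation, then a half-intensity Doppler node.
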SoundCue /Game/Sounds/Fort_Weapons/pistol/FlareGun/SC_FlareGun_MarkingScan_Loop
root=Mixer_0
nodes [
 {id=Doppler_0, children [Enveloper_0], properties {dopplerIntensity=0.5}}
 {
  id=Enveloper_0
  children [WavePlayer_3]
  properties {
   bLoopIndefinitely=true
   durationAfterLoop=2
   loopEnd=4
   pitchMax=1.02
   pitchMin=0.98
   volumeMax=1.05
   volumeMin=0.95
   pitchCurve {editorCurveData {keys [{time=0}]}}
   volumeCurve {editorCurveData {keys [
      {arriveTangent=-1.7322, interpMode=RCIM_Cubic, leaveTangent=-0.6219, tangentMode=RCTM_User, time=0, value=0.005}
      {arriveTangent=2.6258, interpMode=RCIM_Cubic, leaveTangent=3.0142, tangentMode=RCTM_User, time=0.5, value=1}
     ]}}
  }
 }
 {id=Mixer_0, children [Doppler_0,Random_1], properties {inputVolume [1,1]}}
 {id=Random_1, children [WavePlayer_7,WavePlayer_8], properties {weights [1,1]}}
 {id=WavePlayer_3, properties {bLooping=true, soundWaveAssetPtr=&/FlareGun_MarkingScan_Loop_01}}
 {id=WavePlayer_7, properties {bLooping=true, soundWaveAssetPtr=&/FlareGun_InAirLoop_01}}
 {id=WavePlayer_8, properties {bLooping=true, soundWaveAssetPtr=&/FlareGun_InAirLoop_02}}
]
properties {
 attenuationSettings=/Game/Sounds/Attenuation/Fort_Weapon_Projectile_FlareGun_Alert_Att
 maxDistance=10000
 soundClassObject=&Fort_WeaponRanged_Fire_Distant
 volumeMultiplier=0.5957
 attenuationOverrides {
  bAttenuateWithLPF=true
  bEnableLogFrequencyScaling=true
  bEnableReverbSend=false
  distanceAlgorithm=Logarithmic
  falloffDistance=12500
  hPFFrequencyAtMax=100
  lPFFrequencyAtMax=1200
  lPFRadiusMax=12500
  lPFRadiusMin=10000
  spatializationAlgorithm=SPATIALIZATION_HRTF
  stereoSpread=0
  attenuationShapeExtents {x=0}
 }
}
