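# Sound cue for the flare gun's tail-stop sound.
# A random node picks one of three TailStop wave variations with equal weight.
# It borrows a large/medium fire-loop attenuation asset from the Apollo UFO crash site ambience rather than a weapon-specific one.
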SoundCue /Game/Sounds/Fort_Weapons/pistol/FlareGun/SC_FlareGun_TailStop
root=Random_0
nodes [
 {id=Random_0, children [WavePlayer_0,WavePlayer_1,WavePlayer_2], properties {weights [1,1,1]}}
 {id=WavePlayer_0, properties {soundWaveAssetPtr=&/FlareGun_TailStop_001}}
 {id=WavePlayer_1, properties {soundWaveAssetPtr=&/FlareGun_TailStop_003}}
 {id=WavePlayer_2, properties {soundWaveAssetPtr=&/FlareGun_TailStop_002}}
]
properties {attenuationSettings=/Game/Sounds/Apollo/Biomes/UFO_Crash_Sites/FireLoops/Fire_Large_Medium_Att, maxDistance=3500, volumeMultiplier=1}
